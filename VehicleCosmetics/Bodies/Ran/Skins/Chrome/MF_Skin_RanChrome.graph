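# Material function for the Chrome skin of the Ran vehicle body, outputting a full material attribute set.
# RGBA masks and AO blend TeamColor and CustomColor, roughness and Fresnel-based metallic or rubber finishes.
# Reflection-vector noise lookups build a view-dependent sheen without scene reflections.
# Emissive combines masked team, custom and glow colors with audio beat-sync bands and lens flares built from six radially faded, partly spinning mask samples on UV1.
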MaterialFunction /VehicleCosmetics/Bodies/Ran/Skins/Chrome/MF_Skin_RanChrome
graph {nodes [
  n1 = Constant3Vector() {constant=<rgba 0 0 1 0>}
  n2 = MaterialFunctionCall!(Fresnel Power=Constant() {r=1}, Normal=@n1) {materialFunction=/VehicleCosmetics/Materials/MaterialFunctions/MF_Fresnel_Prime}
  n3 = Power(Base=OneMinus(None=@n2.Result)) {constExponent=4, desc=finishIntFalloff}
  Clamp(None=@n3)
  n4 = SetMaterialAttributes!(
   MaterialAttributes
   BaseColor=NamedRerouteUsage()
   Metallic=OneMinus(None=NamedRerouteUsage())
   Roughness=NamedRerouteUsage()
   EmissiveColor=Multiply(A=NamedRerouteUsage(), B=ScalarParameter() {defaultValue=5, parameterName=EmissiveIntensity})
   Normal=NamedRerouteUsage()
   AmbientOcclusion
  ) {attributeSetTypes [
    #1
    #2
    #3
    #4
    #5
    #6
   ]}
  FunctionOutput(None=@n4)
  n5 = MaterialFunctionCall(Fresnel Power=Constant() {r=1}, Normal=@n1) {materialFunction=&MF_Fresnel_Prime}
  n6 = Power(Base=OneMinus(None=@n5.Result)) {constExponent=15, desc=MetallicFinishIntFalloff}
  Multiply(A=@n6)
  n7 = Fresnel() {baseReflectFraction=0, exponent=2}
  n8 = ConstantBiasScale(None=@n7) {bias=0.1, scale=5}
  n9 = TextureCoordinate() {uTiling=0.65}
  n10 = ComponentMask(None=@n9) {g=true}
  n11 = ComponentMask(None=@n9) {r=true}
  n12 = OneMinus(None=TextureSample(UVs=AppendVector(A=@n10, B=@n11)) {samplerType=SAMPLERTYPE_Normal, texture=/VehicleCosmetics/Textures/UniversalTextures/Universal_Noise_Normal_30}.G)
  Multiply(A=@n8, B=Saturate(None=ConstantBiasScale(None=@n12) {bias=-0.5, scale=2}))
  n13 = MaterialFunctionCall!(FakeBeatIntensity, BeatSyncIntensity=ScalarParameter() {defaultValue=1, parameterName=BeatSyncIntensity}) {materialFunction=/VehicleCosmetics/SharedMaterials/MaterialFunctions/MF_AudioBeatSync}
  n14 = Add(A=ComponentMask(None=@n13.Spectrum) {r=true}, B=ComponentMask(None=@n13.Spectrum) {g=true})
  Multiply(A=ScalarParameter() {defaultValue=2, parameterName=BeatStrength}, B=@n14)
  NamedRerouteDeclaration(None=@n14) {name=LowFrequencyBeatync, nodeColor=<rgba 0 0.3672 1 1>}
  n15 = ComponentMask(None=@n13.Spectrum) {b=true}
  n16 = ComponentMask(None=@n13.Spectrum) {a=true}
  NamedRerouteDeclaration(None=Add(A=Max(A=@n15, B=@n16), B=@n14)) {name=BeatsyncMid, nodeColor=<rgba 0.6875 1 0 1>}
  n17 = TextureSampleParameter2D() {desc=UniversalUVs, parameterName=RGBA_Masks, texture=&/T_Ran_chrome_G}
  n18 = NamedRerouteDeclaration(None=StaticComponentMaskParameter(None=TextureSampleParameter2D() {parameterName=AOMap, texture=&/T_Ran_chrome_AO}.RGB) {defaultR=true, desc="R= 01, G=02", parameterName=AO Mask channel}) {name=AO, nodeColor=<rgba 0 1 0.0156 1>}
  NamedRerouteDeclaration(None=LinearInterpolate(B=LinearInterpolate(Alpha=@n17.R) {constA=1, constB=0.01}, Alpha=@n18) {constA=0.2, constB=0.01}) {name=roughness, nodeColor=<rgba 0.9922 0 1 1>}
  n19 = VectorParameter() {parameterName=GlowColor, defaultValue=<rgba 0.2376 0.3331 1 1>}
  NamedRerouteDeclaration(None=@n19.RGB) {name=Normal, nodeColor=<rgba 0 1 0.6719 1>}
  n20 = VectorParameter() {parameterName=TeamColor, defaultValue=<rgba 0.7514 0.5852 0.411 0>}
  n21 = VectorParameter() {parameterName=CustomColor, defaultValue=<rgba 0.7454 0.5776 0.402 0>}
  NamedRerouteDeclaration(None=LinearInterpolate(A=LinearInterpolate(B=@n20.A, Alpha=@n17.R), B=@n21.A, Alpha=@n17.G)) {name=FinishMask, nodeColor=<rgba 0 1 0.1797 1>}
  n22 = Saturate(None=@n6)
  n23 = Saturate(None=Add(A=Power(Base=@n5.Result) {constExponent=5, desc=MetallicFinishOuterFalloff}) {constB=0.1})
  n24 = Add(
   A=Multiply(A=Add(A=@n22, B=@n23)) {constB=0.5}
   B=Multiply(A=Power(Base=@n22) {constExponent=15, desc=MetallicFinishIntFalloff}) {constB=0.5, desc=Metallic Highlight}
  )
  n25 = Power(Base=@n2.Result) {constExponent=5, desc=FinishOuterFalloff}
  n26 = Add(A=Saturate(None=Multiply(A=@n3) {constB=0.25}), B=Clamp(None=Add(A=@n25) {constB=0.2}) {maxDefault=1.5})
  NamedRerouteDeclaration(None=LinearInterpolate(A=Saturate(None=@n24) {desc=Metallic Finish}, B=Saturate(None=@n26) {desc=RubberFinish}, Alpha=NamedRerouteUsage())) {name=Finish, nodeColor=<rgba 1 0 0.1406 1>}
  n27 = Multiply(A=@n8) {constB=0.05}
  n28 = ConstantBiasScale(None=ReflectionVectorWS()) {bias=0}
  n29 = AppendVector(
   A=TextureSample(UVs=ComponentMask(None=@n28) {g=true, r=true}) {texture=/VehicleCosmetics/Textures/UniversalTextures/Universal_Noise_47_RGB}.R
   B=Constant() {r=0.6}
  )
  n30 = Add(A=@n27, B=@n29)
  n31 = ConstantBiasScale(None=Multiply(A=OneMinus(None=@n7), B=TextureSample(UVs=@n30) {texture=&Universal_Noise_47_RGB}.B)) {bias=-0.1, scale=30}
  n32 = Multiply(A=@n27) {constB=0.5}
  n33 = ConstantBiasScale(None=TextureSample(UVs=Add(A=@n32, B=@n29)) {texture=&Universal_Noise_47_RGB}.B) {bias=-0.25, scale=50}
  NamedRerouteDeclaration(None=Max(A=Clamp(None=@n31) {maxDefault=0.25}, B=Saturate(None=@n33))) {name=Sheen, nodeColor=<rgba 0 0.4609 1 1>}
  n34 = NamedRerouteUsage()
  n35 = Subtract(A=Multiply(A=@n20.RGB, B=@n34)) {constB=0.025}
  n36 = NamedRerouteUsage()
  n37 = LinearInterpolate(B=LinearInterpolate(A=Saturate(None=@n35), B=@n20.RGB, Alpha=@n36), Alpha=@n17.R)
  n38 = Multiply(A=@n34, B=@n21.RGB)
  n39 = LinearInterpolate(A=Saturate(None=Subtract(A=@n38) {constB=0.02}), B=@n21.RGB, Alpha=@n36)
  n40 = ConstantBiasScale(None=@n18)
  NamedRerouteDeclaration(None=LinearInterpolate(B=LinearInterpolate(A=@n37, B=@n39, Alpha=@n17.G), Alpha=Saturate(None=@n40))) {name=BaseColor, nodeColor=<rgba 0.3359 0 1 1>}
  n41 = MaterialFunctionCall(
   Fresnel Power=Constant() {r=7}
   Normal=TextureSampleParameter2D() {parameterName=NormalMap, samplerType=SAMPLERTYPE_Normal, texture=&/T_Ran_chrome_G_N}.RGB
  ) {materialFunction=&MF_Fresnel_Prime}
  n42 = NamedRerouteUsage()
  n43 = LinearInterpolate(
   B=LinearInterpolate(A=@n20.RGB, B=@n21.RGB, Alpha=@n17.G)
   Alpha=Multiply(A=Multiply(A=@n17.R, B=@n41.Result), B=ConstantBiasScale(None=@n42))
  )
  n44 = OneMinus(None=@n17.R)
  n45 = TextureSampleParameter2D() {constCoordinate=1, desc=UniversalUVs, parameterName=RGBA_Masks, texture=&T_Ran_chrome_G}
  n46 = Max(A=@n41.Result, B=@n45.B)
  n47 = Multiply(A=Saturate(None=Multiply(A=@n44, B=@n46)), B=NamedRerouteUsage()) {desc=BeatSync Goes Here}
  n48 = Constant() {r=0.15}
  n49 = NamedRerouteUsage()
  n50 = Saturate(None=Subtract(A=@n49) {constB=0.5})
  n51 = NamedRerouteUsage()
  n52 = Clamp(None=@n49) {maxDefault=1.5}
  n53 = Saturate(None=Multiply(A=OneMinus(None=@n52)))
  n54 = Saturate(None=Subtract(A=Subtract(A=@n51, B=@n53)) {constB=0.01})
  NamedRerouteDeclaration(None=LinearInterpolate(A=LinearInterpolate(A=@n43, B=@n19.RGB, Alpha=@n47), B=Desaturation(None=@n20.RGB, Fraction=@n48), Alpha=Multiply(A=@n50, B=@n54)) {desc=Lens Flares go in alpha here}) {name=Emissive, nodeColor=<rgba 1 0 0.3047 1>}
  n55 = TextureCoordinate() {coordinateIndex=1}
  n56 = Constant2Vector() {g=-0.65, r=-0.45}
  n57 = Rotator(Coordinate=Multiply(A=Add(A=@n55, B=@n56)) {constB=-4}) {speed=0.3}
  n58 = ConstantBiasScale(None=@n57) {bias=-0.5, scale=2}
  n59 = OneMinus(None=DotProduct(A=@n58, B=@n58))
  n60 = Multiply(
   A=Saturate(None=Add(A=@n59) {constB=0.2})
   B=TextureSampleParameter2D(UVs=@n57) {desc=UniversalUVs, parameterName=RGBA_Masks, texture=&T_Ran_chrome_G}.A
  )
  n61 = TextureCoordinate() {coordinateIndex=1}
  n62 = Constant2Vector() {g=-0.45, r=-0.8}
  n63 = Rotator(Coordinate=Multiply(A=Add(A=@n61, B=@n62)) {constB=-3})
  n64 = ConstantBiasScale(None=@n63) {bias=-0.5, scale=2}
  n65 = OneMinus(None=DotProduct(A=@n64, B=@n64))
  n66 = Multiply(
   A=Saturate(None=Add(A=@n65) {constB=0.2})
   B=TextureSampleParameter2D(UVs=@n63) {desc=UniversalUVs, parameterName=RGBA_Masks, texture=&T_Ran_chrome_G}.A
  )
  n67 = TextureCoordinate() {coordinateIndex=1}
  n68 = Constant2Vector() {g=-0.8, r=-0.7}
  n69 = Rotator(Coordinate=Multiply(A=Add(A=@n67, B=@n68)) {constB=-7}) {speed=-0.1}
  n70 = ConstantBiasScale(None=@n69) {bias=-0.5, scale=2}
  n71 = Add(A=OneMinus(None=DotProduct(A=@n70, B=@n70))) {constB=0.2}
  n72 = TextureCoordinate() {coordinateIndex=1}
  n73 = Constant2Vector() {g=-0.3, r=-0.3}
  n74 = Rotator(Coordinate=Multiply(A=Add(A=@n72, B=@n73)) {constB=8}) {speed=-0.5}
  n75 = ConstantBiasScale(None=@n74) {bias=-0.5, scale=2}
  n76 = Add(A=OneMinus(None=DotProduct(A=@n75, B=@n75))) {constB=0.2}
  n77 = Max(
   A=Multiply(
    A=Saturate(None=@n71)
    B=TextureSampleParameter2D(UVs=@n69) {desc=UniversalUVs, parameterName=RGBA_Masks, texture=&T_Ran_chrome_G}.A
   )
   B=Multiply(
    A=Saturate(None=@n76)
    B=TextureSampleParameter2D(UVs=@n74) {desc=UniversalUVs, parameterName=RGBA_Masks, texture=&T_Ran_chrome_G}.A
   )
  ) {constB=0}
  n78 = TextureCoordinate() {coordinateIndex=1}
  n79 = Constant2Vector() {g=-0.5, r=-0.76}
  n80 = Rotator(Coordinate=Multiply(A=Add(A=@n78, B=@n79)) {constB=-8})
  n81 = ConstantBiasScale(None=@n80) {bias=-0.5, scale=2}
  n82 = Add(A=OneMinus(None=DotProduct(A=@n81, B=@n81))) {constB=0.2}
  n83 = TextureCoordinate() {coordinateIndex=1}
  n84 = Constant2Vector() {g=-0.95, r=-0.5}
  n85 = Rotator(Coordinate=Multiply(A=Add(A=@n83, B=@n84)) {constB=-5}) {speed=0.5}
  n86 = ConstantBiasScale(None=@n85) {bias=-0.5, scale=2}
  n87 = Add(A=OneMinus(None=DotProduct(A=@n86, B=@n86))) {constB=0.2}
  n88 = Max(
   A=Multiply(
    A=Saturate(None=@n82)
    B=TextureSampleParameter2D(UVs=@n80) {desc=UniversalUVs, parameterName=RGBA_Masks, texture=&T_Ran_chrome_G}.A
   )
   B=Multiply(
    A=Saturate(None=@n87)
    B=TextureSampleParameter2D(UVs=@n85) {desc=UniversalUVs, parameterName=RGBA_Masks, texture=&T_Ran_chrome_G}.A
   )
  ) {constB=0}
  NamedRerouteDeclaration(None=Max(A=Max(A=@n60, B=@n66) {constB=0}, B=Max(A=@n77, B=@n88) {constB=0})) {name=LensFlares, nodeColor=<rgba 0 0.625 1 1>}
 ]}
